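# Niagara idle effect that adds a pair of glowing purple eyes to the Clawed Raven backpack cosmetic.
# A single emitter bursts two fixed-size HDR-colored sprites, then partitions them by execution index to set mirrored eye positions, with a small camera offset.
# The particles persist until the emitter deactivates, and a front-end float modifier lowers their alpha in the front end compared to gameplay.
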NiagaraSystem /BRCosmetics/Accessories/FORT_Backpacks/Backpack_F_MED_ClawedRaven/FX/NS_Backpack_ClawedRaven_EyeGlow_Idle
dependencies {usedDataInterfaces [], usedDynamicInputs [], usedModules [&SystemState]}
emitters {EyeGlows {
  emitterUpdateScript [
   &EmitterState(
    Inactive Response=::NewEnumerator1
    Life Cycle Mode=::NewEnumerator1
    Loop Behavior=::NewEnumerator1
    Loop Duration Mode=::NewEnumerator1
    UseLoopDelay=false
   )
   &SpawnBurst_Instantaneous(Spawn Count=2, Use Spawn Probability=false, UseLoopCountLimit=true)
  ]
  particleSpawnScript [
   &InitializeParticle(
    Color=<rgba 4.2333 0.4 5 0.25>
    Sprite Size Mode=::NewEnumerator3
    Uniform Sprite Size=7
    UsePositionOffset=false
    Write Material Random=true
   )
   &PartitionParticles(ID to Partition:NiagaraInt32=&ReturnExecIndex())
   &CameraOffset(Camera Offset Amount=2)
   /SetVariables(Particles.Position=&MaskVectorBySpawnGroup(Masked Vector=<-3.05 -12 9.45>, Passthrough Vector=<3.05 -12 9.45>, Spawn Group Mask=$Particles.PartitionParticles.Partition)) {isSetParameters=true}
  ]
  particleUpdateScript [
   &ParticleState(Kill Particles When Lifetime Has Elapsed=false, Let Infinitely Lived Particles Die When Emitter Deactivates=true)
   &ScaleColor(Scale Alpha=&FN_FrontEndFloatModifier(Gameplay=1, Front End=0.6), ScaleA=true, ScaleRGB=false)
  ]
  rendererClasses [&NiagaraSpriteRendererProperties]
  renderers [{rendererClass=&NiagaraSpriteRendererProperties}]
 }}
summary {
 systemName=NS_Backpack_ClawedRaven_EyeGlow_Idle
 emitters [{emitterName=EyeGlows, rendererClasses [&NiagaraSpriteRendererProperties]}]
 userVariables [
  User.bIsFrontEnd {type=NiagaraBool}
  User.bIsFrontEndPreview {type=NiagaraBool}
  User.bKillReaction {type=NiagaraBool}
  User.global Alpha {type=NiagaraFloat, defaultValue=0, description=""}
  User.IsSkydiving {type=NiagaraBool}
  User.showFX {type=NiagaraFloat, defaultValue=1, description=""}
 ]
}
