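# Physics asset for the F_MED_ClawedRaven backpack cosmetic, defining collision bodies and joints for its dangling keychain.
# Each keychain bone, from the master through three links, has a capsule body, and a larger capsule covers the pack.
# Constraints chain the links back to the master bone with moderate swing limits and tighter twist limits.
# No constraint joins the pack body to the keychain.
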PhysicsAsset /BRCosmetics/Accessories/FORT_Backpacks/Backpack_F_MED_ClawedRaven/Meshes/F_MED_ClawedRaven_Pack_Physics
bodies {
 dyn_keychain_01 [{shapeType=Capsule, length=1, radius=1.5, shapeName=dyn_keychain_01_capsule, center=<2 0 0>, rotation=<rot -90 -90 90.0002>}]
 dyn_keychain_02 [{shapeType=Capsule, length=1, radius=1.5, shapeName=dyn_keychain_02_capsule, center=<2 0 0>, rotation=<rot -90 -90 90.0002>}]
 dyn_keychain_03 [{shapeType=Capsule, length=0.5, radius=3, shapeName=dyn_keychain_03_capsule, center=<3.5 0 0>, rotation=<rot -90 -90 90.0002>}]
 dyn_keychain_MSTR [{shapeType=Capsule, length=0.5, radius=0.5, shapeName=dyn_keychain_MSTR_capsule}]
 pack [{shapeType=Capsule, length=5, radius=10, shapeName=pack_capsule, center=<0 -7.4089 -2.6494>, rotation=<rot -90 0 -0.0001>}]
}
constraints [
 {bone1Name=dyn_keychain_03, bone2Name=dyn_keychain_02, swing1LimitDegrees=35, swing2LimitDegrees=35, twistLimitDegrees=10}
 {bone1Name=dyn_keychain_02, bone2Name=dyn_keychain_01, swing1LimitDegrees=35, swing2LimitDegrees=35, twistLimitDegrees=10}
 {bone1Name=dyn_keychain_01, bone2Name=dyn_keychain_MSTR, swing1LimitDegrees=35, swing2LimitDegrees=35, twistLimitDegrees=10}
]
